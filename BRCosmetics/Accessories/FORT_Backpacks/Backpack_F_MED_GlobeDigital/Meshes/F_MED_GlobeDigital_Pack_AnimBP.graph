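# Animation Blueprint for the digital globe back bling that spins its sphere and offsets the pack.
# Each update derives a yaw from game time at a constant rate, so the spin follows the clock rather than accumulated delta time.
# The AnimGraph takes the input pose, applies that yaw additively to the sphere bone in bone space, then adds a fixed bone-space Y translation to the pack bone.
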AnimBlueprint /BRCosmetics/Accessories/FORT_Backpacks/Backpack_F_MED_GlobeDigital/Meshes/F_MED_GlobeDigital_Pack_AnimBP
target_skeleton=&/F_MED_GlobeDigital_Pack_Skeleton
graphs [
 AnimGraph {nodes [
   n1 = "Transform (Modify) Bone - Bone: sphere"(
    ComponentPose=Local To Component(LocalPose=Input Pose())
    Alpha=1
    Translation="0,0,0"
    Rotation=Get Sphere_Rotation_Combine() -> (Sphere_Rotation_Combine:Rotator)
    Scale="1,1,1"
   ) {rotationMode=BMM_Additive, rotationSpace=BCS_BoneSpace, alphaBoolBlend {blendOption=Linear}, boneToModify {boneName=sphere}}
   n2 = "Transform (Modify) Bone - Bone: pack"(ComponentPose=@n1, Translation="0,8.8,0", Rotation="0,0,0") {
    rotationMode=BMM_Additive
    rotationSpace=BCS_BoneSpace
    translationMode=BMM_Additive
    translationSpace=BCS_BoneSpace
    alphaBoolBlend {blendOption=Linear}
    boneToModify {boneName=pack}
    translation {y=8.8}
   }
   Output Pose(Result=Component To Local(ComponentPose=@n2)) {layerGroup=None}
  ]}
 EventGraph {nodes [
   Event BlueprintUpdateAnimation()
   > Set Sphere_Rotation(Sphere_Rotation:double=float * float(A=GetGameTimeInSeconds(), B=30) {is=title K2Node_PromotableOperator}) -> (Output_Get:double)
   > Set Sphere_Rotation_Combine(Sphere_Rotation_Combine:Rotator=MakeRotator(Yaw=Get Sphere_Rotation() -> (Sphere_Rotation:double))) -> (Output_Get:Rotator)
   TryGetPawnOwner()
  ]}
]
variables [
 Sphere_Rotation {type=number}
 Sphere_Rotation_Combine {type=Rotator}
]
